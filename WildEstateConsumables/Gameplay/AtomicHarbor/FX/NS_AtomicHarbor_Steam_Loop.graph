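# Niagara looping steam effect for Atomic Harbor, a single emitter adapted from reload smoke that draws soft, persistent vertical wisps.
# A burst is partitioned into five short ribbons, each lerped from a jittered base to a seeded random height, with particles kept alive past their lifetime.
# Link-order curves shape width, alpha and material parameters; alpha also scales with time of day, and a user flag enables emissive parameters.
# Motion forces, skeletal mesh sampling and Shorten-driven raise logic are disabled.
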NiagaraSystem /WildEstateConsumables/Gameplay/AtomicHarbor/FX/NS_AtomicHarbor_Steam_Loop
dependencies {usedDataInterfaces [&NiagaraDataInterfaceCurve], usedDynamicInputs [], usedModules [&SystemState]}
emitters {Reload_Smoke001 {
  emitterUpdateScript [
   &EmitterState(Life Cycle Mode=::NewEnumerator1, Loop Behavior=::NewEnumerator1, Loop Duration Mode=::NewEnumerator1, UseLoopDelay=false)
   &SpawnBurst_Instantaneous(Spawn Count=25, Use Spawn Probability=false, UseLoopCountLimit=false)
   &SpawnRate(SpawnRate=4, Use Spawn Probability=false) {enabled=false}
   /SetVariables(Emitter.float:NiagaraFloat=&RemapRange(Input Value To Remap=$User.Shorten, Input Min=0, Input Max=1, Output Min=10, Output Max=-100)) {enabled=false, isSetParameters=true}
  ]
  particleSpawnScript [
   &InitializeParticle(
    Color=$User.BaseColorAlpha
    Lifetime=0.7
    Ribbon Facing Mode=::NewEnumerator1
    Ribbon Facing Vector:Vector3f=<1 0 0>
    Ribbon Width=&Multiply_Float(A=15, B=$User.RibbonScale)
    Ribbon Width Mode=::NewEnumerator1
    Sprite Size Mode=::NewEnumerator3
    UsePositionOffset=false
    Write Material Random=true
   )
   &PartitionParticles(
    Number Of Partitions=5
    Particles Per Partition=5
    Partition Mode:ENiagara_PartitionMode_v2=::NewEnumerator1
    Write Partition Index to Ribbon Link Order:ENiagara_AttributeSamplingApplyOutput=::NewEnumerator0
    Write Partition to Ribbon ID=::NewEnumerator0
   )
   &ShapeLocation(
    MaskSpawnGroup=false
    Non Uniform Scale=<2 1 0>
    Offset=&RandomRangeVector(
     Minimum=<-5 -5 0>
     Maximum=<5 5 1>
     Randomness Mode=::NewEnumerator1
     Random Seed=$Particles.PartitionParticles.Partition
     Fixed Random Seed=true
    )
    Offset Mode=::NewEnumerator0
    Sphere Distribution=::NewEnumerator1
    Sphere Radius=15
    U Position=$Particles.PartitionParticles.NormalizedPartition
    UseOwnerScale=true
   )
   &AddVelocity(Velocity=&RandomRangeVector(
     Minimum=<0 0 60>
     Maximum=<0 0 90>
     Randomness Mode=::NewEnumerator1
     Random Seed=$Particles.PartitionParticles.Partition
     Fixed Random Seed=true
    )) {enabled=false}
  ]
  particleUpdateScript [
   &ParticleState(Kill Particles When Lifetime Has Elapsed=false, Let Infinitely Lived Particles Die When Emitter Deactivates=true)
   /SetVariables(Particles.Position=&Lerp_Vector(
     A=Linked() {linkedVariable {name=Particles.ShapeLocation.ShapePosition, type=NiagaraPosition}}
     B=&AddVectorToPosition(
      Position=$Particles.ShapeLocation.ShapePosition
      Vector=&RandomRangeVector(
       Minimum=<0 0 30>
       Maximum=<0 0 100>
       Randomness Mode=::NewEnumerator1
       Random Seed=$Particles.PartitionParticles.Partition
       Fixed Random Seed=true
      )
     )
     Alpha=$Particles.PartitionParticles.Normalized Index In Partition
    )) {isSetParameters=true}
   &ScaleColor(
    Scale Alpha=&Multiply_Float(
     A=&FloatFromCurve(
      FloatCurve=DataInterface() {propertyValues {Curve {
         defaultValue=0
         keys [
          {arriveTangent=-0.0628, interpMode=RCIM_Cubic, leaveTangent=-0.0628, tangentMode=RCTM_User, time=0, value=0}
          {arriveTangent=-2.5249, interpMode=RCIM_Cubic, leaveTangent=-2.5249, tangentMode=RCTM_User, time=0.1408, value=0.9847}
          {arriveTangent=-0.0405, interpMode=RCIM_Cubic, leaveTangent=-0.0405, tangentMode=RCTM_User, time=0.5, value=0}
         ]
        }}}
      CurveIndex=$Particles.Initial.RibbonLinkOrder
      Scale Curve=&FNTimeOfDayScalar(Night Scale=1, Morning Scale=0.3, Evening Scale=0.5)
     )
     B=&FloatFromCurve(
      FloatCurve=DataInterface() {propertyValues {Curve {keys [<at 0 1>,<at 0.3 1>,<at 0.85 1>,<at 1 0>]}}}
      CurveIndex=$Particles.Initial.RibbonLinkOrder
     )
    )
    Scale RGB=&MakeVectorFromLinearColorRGB(Color=&MakeCustomLinearColorFromBool(
      Bool=$User.UseEmissiveSmoke
      True Color=&LerpLinearColors(
       StartColor=$User.EmissiveColor
       EndColor=&MultiplyLinearColor(Linear Color=$User.EmissiveColor, Scale Factor=&MakeVector4_FromFloat(Float=0.25))
       LerpFactor=&FloatFromCurve(
        FloatCurve=DataInterface() {propertyValues {Curve {keys [<at 0 0 RCIM_Cubic>, {arriveTangent=2.2656, interpMode=RCIM_Cubic, leaveTangent=2.2656, tangentMode=RCTM_User, time=1, value=1}]}}}
        CurveIndex=$Particles.Initial.RibbonLinkOrder
       )
      )
      False Color=<rgba 1 1 1 1>
     ))
    ScaleA=true
    ScaleRGB=false
   )
   &DynamicMaterialParameters(
    Index 0 Param 1=&FloatFromCurve(
     FloatCurve=DataInterface() {propertyValues {Curve {keys [
         <at 0 1 RCIM_Cubic>
         {arriveTangent=-1.7387, arriveTangentWeight=1.9926, interpMode=RCIM_Cubic, leaveTangent=-1.7387, tangentMode=RCTM_User, time=1, value=0.125}
        ]}}}
     CurveIndex=$Particles.Initial.RibbonLinkOrder
    )
    Index 0 Param 3=&RandomRangeFloat(Minimum=-0.4, Maximum=-0.6, Randomness Mode=::NewEnumerator1, Random Seed=$Particles.PartitionParticles.Partition, Fixed Random Seed=true)
    Index 0 Param 4=&MakeCustomFloatFromBool(
     Bool=$User.UseEmissiveSmoke
     True Float=&FloatFromCurve(
      FloatCurve=DataInterface() {propertyValues {Curve {keys [{arriveTangent=-2.94, interpMode=RCIM_Cubic, time=0, value=1}, <at 0.75 0 RCIM_Cubic>]}}}
      CurveIndex=$Particles.Initial.RibbonLinkOrder
     )
     False Float=0
    )
    Index 1 Param 1=&FloatFromCurve(FloatCurve=DataInterface() {propertyValues {Curve {keys [<at 0 1 RCIM_Cubic>]}}}, CurveIndex=$Particles.Initial.RibbonLinkOrder)
    Index 1 Param 2=&FloatFromCurve(
     FloatCurve=DataInterface() {propertyValues {Curve {keys [{arriveTangent=-2.2405, interpMode=RCIM_Cubic, leaveTangent=-2.2405, tangentMode=RCTM_User, time=0, value=1}, <at 1 0 RCIM_Cubic>]}}}
     CurveIndex=$Particles.Initial.RibbonLinkOrder
    )
    Module.Param0WriteEnabled=true
    Module.Param0WriteEnabled1=true
    Module.Param1WriteEnabled=false
    Module.Param1WriteEnabled1=true
    Module.Param2WriteEnabled=true
    Module.Param2WriteEnabled1=false
    Module.Param3WriteEnabled=true
    Module.Param3WriteEnabled1=false
    Write Parameter Index 1=true
   )
   &ScaleRibbonWidth(Ribbon Width Scale=&FloatFromCurve(
     FloatCurve=DataInterface() {propertyValues {Curve {keys [{tangentMode=RCTM_User, time=0, value=1}, {tangentMode=RCTM_User, time=1, value=0}]}}}
     CurveIndex=$Particles.Initial.RibbonLinkOrder
     Scale Curve=4
    ))
   &SkeletalMeshLocation(
    Bone Sampling Mode=::NewEnumerator4
    MaskSpawnGroup=false
    Orientation Sampling=::NewEnumerator1
    Position Sampling=::NewEnumerator2
    Skeletal Mesh=$System.Skeletal Mesh
   ) {enabled=false}
   &SkeletalMeshLocation(
    Bone Sampling Mode=::NewEnumerator4
    MaskSpawnGroup=false
    Orientation Sampling=::NewEnumerator1
    Position Sampling=::NewEnumerator2
    Skeletal Mesh=$User.PlayerPawn
   ) {enabled=false, module=SkeletalMeshLocation001}
   &PointAttractionForce(
    Attraction Radius=500
    Attraction Strength=&FloatFromCurve(FloatCurve=DataInterface() {propertyValues {Curve {keys [<at 0 1>,<at 1 0>]}}}, Scale Curve=100)
    Attractor Position Offset:Vector3f=<0 70 100>
    Attractor Position Offset Coordinate Space:ENiagaraCoordinateSpace=::Local
    Falloff Exponent=0.2
    Kill Within Radius=false
    Module.Use Falloff=true
    UseAttractorOffset=true
   ) {enabled=false}
   &CurlNoiseForce(
    Mask Curl Noise=true
    Noise Frequency=10
    Noise Quality / Cost=::NewEnumerator0
    Noise Strength=625
    Pan Noise=true
    Pan Noise Field=<-10 0 0>
   ) {enabled=false}
   &InheritSourceMovement(
    Applied Position Scale=&Multiply_VectorByFloat(Vector=<1 1 1>, Float=1)
    ApplyForce=false
    ApplyPosition=true
    ApplyPositionFalloff=false
   ) {enabled=false}
   &AccelerationForce(Acceleration=&MakeVector(
     X=0
     Y=0
     Z=&FloatFromCurve(FloatCurve=DataInterface() {propertyValues {Curve {keys [<at 0 -50>,<at 0.4 -100>,<at 0.5 400>]}}})
    )) {enabled=false}
   /SetVariables(Particles.Position=&Lerp_Vector(A=$Particles.SkeletalMeshLocation.SampledPosition, B=$Particles.SkeletalMeshLocation001.SampledPosition, Alpha=$Particles.NormalizedAge)) {enabled=false, isSetParameters=true}
   /SetVariables(Particles.Position=&AddVectorToPosition(
     Position=$Particles.Position
     Vector=&MakeVector(
      X=0
      Y=0
      Z=&FloatFromCurve(
       FloatCurve=DataInterface() {propertyValues {Curve {keys [
           {interpMode=RCIM_Cubic, tangentMode=RCTM_User, time=0, value=1}
           {interpMode=RCIM_Cubic, tangentMode=RCTM_User, time=0.5, value=1}
           {tangentMode=RCTM_User, time=1, value=0}
          ]}}}
       Scale Curve=&Add_Float(
        A=&Multiply_Float(
         A=&FloatFromCurve(FloatCurve=DataInterface() {propertyValues {Curve {keys [
              {interpMode=RCIM_Cubic, tangentMode=RCTM_User, time=0, value=0}
              {interpMode=RCIM_Cubic, tangentMode=RCTM_User, time=0.5, value=1}
              {tangentMode=RCTM_User, time=1, value=0}
             ]}}})
         B=&Multiply_Float(
          A=$Emitter.ZRaiseEmitter
          B=&Waveform(
           Waveform Count=::NewEnumerator1
           "Waveform [1]"=::NewEnumerator0
           "[1] Frequency"=0.1
           "[1] Amplitude Min/Max"=<0.3 1>
           "[2] Average"=false
           "[2] Input Position"=0
           "[2] Phase"=$Emitter.Age
           "[2] Phase Scale"=1
           "[2] Compound Frequency"=2
           "[2] Frequency"=0.3
           "[2] Amplitude Min/Max"=<0.3 1>
           "[2] Scale"=1
           "[2] Bias"=0
           "[2] Exponent"=1
           "[2] Pulse Duration"=0.1
           "[2] Clip Triangle Min/Max"=<0 0>
           "[2] Random Seed"=0
          )
         )
        )
        B=&Add_Float(
         A=&Multiply_Float(
          A=&FloatFromCurve(FloatCurve=DataInterface() {propertyValues {Curve {keys [
               {interpMode=RCIM_Cubic, tangentMode=RCTM_User, time=0.3, value=0}
               {interpMode=RCIM_Cubic, tangentMode=RCTM_User, time=0.6706, value=1}
               {tangentMode=RCTM_User, time=1, value=0}
              ]}}})
          B=&Multiply_Float(
           A=$Emitter.ZRaiseEmitter
           B=&Waveform(
            Waveform Count=::NewEnumerator1
            "Waveform [1]"=::NewEnumerator0
            "[1] Frequency"=0.05
            "[1] Amplitude Min/Max"=<-0.4 0.3>
            "[2] Average"=false
            "[2] Input Position"=0
            "[2] Phase"=$Emitter.Age
            "[2] Phase Scale"=1
            "[2] Compound Frequency"=2
            "[2] Frequency"=0.15
            "[2] Amplitude Min/Max"=<-0.4 0.3>
            "[2] Scale"=1
            "[2] Bias"=0
            "[2] Exponent"=1
            "[2] Pulse Duration"=0.1
            "[2] Clip Triangle Min/Max"=<0 0>
            "[2] Random Seed"=0
           )
          )
         )
         B=&Multiply_Float(
          A=&FloatFromCurve(FloatCurve=DataInterface() {propertyValues {Curve {keys [
               {interpMode=RCIM_Cubic, tangentMode=RCTM_User, time=0, value=0}
               {interpMode=RCIM_Cubic, tangentMode=RCTM_User, time=0.1441, value=1}
               {tangentMode=RCTM_User, time=0.463, value=0}
              ]}}})
          B=10
         )
        )
       )
      )
     )
    )) {enabled=false, isSetParameters=true}
   &InheritSourceMovement(Applied Position Scale=<0 0 100>, ApplyForce=false, ApplyPosition=true, ApplyPositionFalloff=false) {enabled=false, module=InheritSourceMovement001}
   &SolveForcesAndVelocity(Clamp Velocity=false, Limit Acceleration=false)
  ]
  rendererClasses [&NiagaraRibbonRendererProperties, &NiagaraSpriteRendererProperties]
  renderers [
   {rendererClass=&NiagaraRibbonRendererProperties}
   {rendererClass=&NiagaraSpriteRendererProperties, rendererIndex=1}
  ]
 }}
summary {
 systemName=NS_AtomicHarbor_Steam_Loop
 emitters [{
   emitterName=Reload_Smoke001
   rendererClasses [&NiagaraRibbonRendererProperties, &NiagaraSpriteRendererProperties]
  }]
 userVariables [
  User.BaseColorAlpha {type=LinearColor, description="", defaultValue {struct=LinearColor, value=<rgba 1 1 1 0.12>}}
  User.EmissiveColor {type=LinearColor, description="", defaultValue {struct=LinearColor, value=<rgba 0.1354 0.0043 0 1>}}
  User.PlayerPawn {
   type=NiagaraDataInterfaceSkeletalMesh
   defaultValue {
    struct=NiagaraExt_VariableValue_DataInterface
    value {dataInterface=&self:NiagaraDataInterfaceSkeletalMesh_0}
   }
  }
  User.RibbonMaterial {
   type=MaterialInterface
   description=""
   defaultValue {
    struct=NiagaraExt_VariableValue_Object
    value {object=/Game/Effects/Niagara/MasterMaterials/Reload/M_Reload_SmokeTrail, objectClass=/Script/Engine.MaterialInterface}
   }
  }
  User.RibbonScale {type=NiagaraFloat, defaultValue=1, description=""}
  User.Shorten {type=NiagaraFloat, defaultValue=0, description=""}
  User.UseEmissiveSmoke {type=NiagaraBool}
 ]
}
